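# UI master material for the Creative property editor's binding icon, drawing two side-by-side chevron shapes procedurally, with no textures.
# Centered, scaled UVs are mirrored vertically and fed to custom parallelogram signed-distance functions, so skewed bars become chevrons.
# Static switches render each side filled or outlined (left outlined by default), and smoothstepped edges combine by max into a mask.
# A further parallelogram mask blends between the left and right color parameters.
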Material /Game/Athena/UI/InGame/Creative/PropertyEditor/Art/M_UI_CreativeBindingIcon_Master
graph {
 properties {
  bUsedWithStaticMesh=false
  blendMode=BLEND_Translucent
  materialDomain=MD_UI
  pixelDepthOffsetMode=PDOM_Legacy
  usedShadingModels=Default Lit
 }
 nodes [
  n1 = NamedRerouteUsage()
  n2 = MaterialFunctionCall(Float2=Add(A=TextureCoordinate(), B=Constant2Vector() {g=-0.5})) {materialFunction=&BreakOutFloat2Components}
  n3 = Multiply(A=@n2.G)
  n4 = MaterialFunctionCall(UVs=AppendVector(A=@n2.R, B=Abs(None=@n3)), Texture Scale=Constant() {r=0.85}) {materialFunction=&ScaleUVsByCenter}
  MaterialExpressionAdd_5 = Add(A=@n4.UVs, B=Constant2Vector() {r=-0.3})
  MaterialExpressionScalarParameter_6 = ScalarParameter() {defaultValue=0.275, parameterName=Width-01}
  MaterialExpressionScalarParameter_7 = ScalarParameter() {defaultValue=-0.15, parameterName=Skew}
  MaterialExpressionScalarParameter_8 = ScalarParameter() {defaultValue=0.275, parameterName=Height-01}
  n5 = Custom!(
   p=@MaterialExpressionAdd_5
   wi=@MaterialExpressionScalarParameter_6
   sk=@MaterialExpressionScalarParameter_7
   he=@MaterialExpressionScalarParameter_8
  ) {
   code="float2  e  = float2(sk,he);\r\nfloat e2 = sk*sk + he*he;\r\np = (p.y<0.0)?-p:p;\r\nfloat2  w = p - e; w.x -= clamp(w.x,-wi,wi);\r\nfloat2  d = float2(dot(w,w), -w.y);\r\nfloat s = p.x*e.y - p.y*e.x;\r\np = (s<0.0)?-p:p;\r\nfloat2  v = p - float2(wi,0); v -= e*clamp(dot(v,e)/e2,-1.0,1.0);\r\nd = min( d, float2(dot(v,v), wi*he-abs(s)));\r\nreturn sqrt(d.x)*sign(-d.y);"
   description=Parallelogram
   outputType=CMOT_Float2
   inputs [
    {inputName=p, input {expression=&self:MaterialExpressionAdd_5}}
    {
     inputName=wi
     input {expression=&self:MaterialExpressionScalarParameter_6}
    }
    {
     inputName=sk
     input {expression=&self:MaterialExpressionScalarParameter_7}
    }
    {
     inputName=he
     input {expression=&self:MaterialExpressionScalarParameter_8}
    }
   ]
  }
  n6 = SmoothStep(Min=Subtract(A=@n1, B=Multiply(A=@n1) {constB=2}), Max=@n1, Value=@n5)
  LinearInterpolate(
   A=VectorParameter() {parameterName=Right-RGBA, defaultValue=<rgba 1 0 0 0>}.RGB
   B=VectorParameter() {parameterName=Left-RGBA, defaultValue=<rgba 0 1 0.8119 0>}.RGB
   Alpha=Saturate(None=ComponentMask(None=@n6) {r=true})
  )
  NamedRerouteDeclaration(None=ScalarParameter() {defaultValue=0.03, parameterName=SmoothStep-Max}) {name=Ratio-of-1px-to-Widget, nodeColor=<rgba 0 0 0 1>}
  NamedRerouteDeclaration(None=ScalarParameter() {defaultValue=-0.05, parameterName=SDF-Falloff}) {name=Ratio-of-1px-to-Widget-Compensated, nodeColor=<rgba 0 0 0 1>}
  n7 = NamedRerouteUsage()
  MaterialExpressionAdd_1 = Add(A=@n4.UVs, B=Constant2Vector() {r=-0.815}) {constB=2}
  MaterialExpressionScalarParameter_1 = ScalarParameter() {defaultValue=0.2, parameterName=Width}
  MaterialExpressionScalarParameter_2 = ScalarParameter() {defaultValue=-0.15, parameterName=Skew}
  MaterialExpressionScalarParameter_3 = ScalarParameter() {defaultValue=0.25, parameterName=Height}
  n8 = Custom(
   p=@MaterialExpressionAdd_1
   wi=@MaterialExpressionScalarParameter_1
   sk=@MaterialExpressionScalarParameter_2
   he=@MaterialExpressionScalarParameter_3
  ) {
   code="float2  e  = float2(sk,he);\r\nfloat e2 = sk*sk + he*he;\r\np = (p.y<0.0)?-p:p;\r\nfloat2  w = p - e; w.x -= clamp(w.x,-wi,wi);\r\nfloat2  d = float2(dot(w,w), -w.y);\r\nfloat s = p.x*e.y - p.y*e.x;\r\np = (s<0.0)?-p:p;\r\nfloat2  v = p - float2(wi,0); v -= e*clamp(dot(v,e)/e2,-1.0,1.0);\r\nd = min( d, float2(dot(v,v), wi*he-abs(s)));\r\nreturn sqrt(d.x)*sign(-d.y);"
   description=Parallelogram
   outputType=CMOT_Float2
   inputs [
    {inputName=p, input {expression=&self:MaterialExpressionAdd_1}}
    {
     inputName=wi
     input {expression=&self:MaterialExpressionScalarParameter_1}
    }
    {
     inputName=sk
     input {expression=&self:MaterialExpressionScalarParameter_2}
    }
    {
     inputName=he
     input {expression=&self:MaterialExpressionScalarParameter_3}
    }
   ]
  }
  n9 = NamedRerouteUsage()
  n10 = Max(A=@n8, B=Multiply(A=Subtract(A=@n8, B=@n9) {constB=-0.05}) {constB=-1})
  n11 = SmoothStep(
   Min=Subtract(A=@n7, B=Multiply(A=@n7) {constB=2})
   Max=@n7
   Value=Multiply(A=StaticSwitchParameter(True=@n10, False=@n8) {parameterName=bRight-Outline}) {constB=-1}
  )
  n12 = NamedRerouteUsage()
  MaterialExpressionAdd_3 = Add(A=@n4.UVs, B=Constant2Vector() {r=-0.3})
  MaterialExpressionScalarParameter_21 = ScalarParameter() {defaultValue=0.2, parameterName=Width}
  MaterialExpressionScalarParameter_22 = ScalarParameter() {defaultValue=-0.15, parameterName=Skew}
  MaterialExpressionScalarParameter_23 = ScalarParameter() {defaultValue=0.25, parameterName=Height}
  n13 = Custom(
   p=@MaterialExpressionAdd_3
   wi=@MaterialExpressionScalarParameter_21
   sk=@MaterialExpressionScalarParameter_22
   he=@MaterialExpressionScalarParameter_23
  ) {
   code="float2  e  = float2(sk,he);\r\nfloat e2 = sk*sk + he*he;\r\np = (p.y<0.0)?-p:p;\r\nfloat2  w = p - e; w.x -= clamp(w.x,-wi,wi);\r\nfloat2  d = float2(dot(w,w), -w.y);\r\nfloat s = p.x*e.y - p.y*e.x;\r\np = (s<0.0)?-p:p;\r\nfloat2  v = p - float2(wi,0); v -= e*clamp(dot(v,e)/e2,-1.0,1.0);\r\nd = min( d, float2(dot(v,v), wi*he-abs(s)));\r\nreturn sqrt(d.x)*sign(-d.y);"
   description=Parallelogram
   outputType=CMOT_Float2
   inputs [
    {inputName=p, input {expression=&self:MaterialExpressionAdd_3}}
    {
     inputName=wi
     input {expression=&self:MaterialExpressionScalarParameter_21}
    }
    {
     inputName=sk
     input {expression=&self:MaterialExpressionScalarParameter_22}
    }
    {
     inputName=he
     input {expression=&self:MaterialExpressionScalarParameter_23}
    }
   ]
  }
  n14 = NamedRerouteUsage()
  n15 = Max(A=@n13, B=Multiply(A=Subtract(A=@n13, B=@n14) {constB=-0.05}) {constB=-1})
  n16 = SmoothStep(
   Min=Subtract(A=@n12, B=Multiply(A=@n12) {constB=2})
   Max=@n12
   Value=Multiply(A=StaticSwitchParameter(True=@n15, False=@n13) {defaultValue=true, parameterName=bLeft-Outline}) {constB=-1}
  )
  Saturate(None=ComponentMask(None=Max(A=@n11, B=@n16)) {g=true})
 ]
}
