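# Translucent UI-domain material for font color in the Creative mode property editor's event UI, judging by its name and path.
# Its graph holds a single TextColor vector parameter in the Colors group, defaulting to an opaque dark gray.
# The file shows no other nodes and no link from the parameter to a material output.
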Material /Game/Athena/UI/InGame/Creative/PropertyEditor/Art/M_UI_CreativePropertyEvent_FontColor
graph {
 properties {
  bUsedWithStaticMesh=false
  blendMode=BLEND_Translucent
  materialDomain=MD_UI
  pixelDepthOffsetMode=PDOM_Legacy
  usedShadingModels=Default Lit
 }
 nodes [VectorParameter() {group=Colors, parameterName=TextColor, sortPriority=2, defaultValue=<rgba 0.0648 0.0648 0.0648 1>}]
}
