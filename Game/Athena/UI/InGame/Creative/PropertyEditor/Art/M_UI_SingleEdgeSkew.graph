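# Translucent UI material that renders a slanted single edge with an anti-aliased feather, for Creative property editor art.
# A dot product of optionally X-mirrored UVs and a SkewAmount vector scaled by a named reroute input forms the skew gradient.
# Pixel Size ratio reroutes and a FeatherMin compensation support the feathering.
# A static switch picks linear subtract/divide or SmoothStep for the soft edge, which is then inverted and saturated.
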Material /Game/Athena/UI/InGame/Creative/PropertyEditor/Art/M_UI_SingleEdgeSkew
graph {
 properties {
  bUsedWithStaticMesh=false
  blendMode=BLEND_Translucent
  materialDomain=MD_UI
  pixelDepthOffsetMode=PDOM_Legacy
  usedShadingModels=Default Lit
 }
 nodes [
  MaterialFunctionCall!(Vector2, MaximumNumberOfDigits, UVs, DebugTextLocation RG_UpperRight BA_LowerLeft, Component Spacing) {materialFunction=/Engine/Functions/Engine_MaterialFunctions02/Utility/DebugFloat2Values}
  MaterialFunctionCall() {materialFunction=&DebugFloat2Values}
  n1 = AppendVector(A=Constant() {r=1}, B=Multiply(A=ScalarParameter() {defaultValue=0.25, parameterName=SkewAmount}, B=NamedRerouteUsage()))
  MaterialFunctionCall(Vector2=@n1) {materialFunction=&DebugFloat2Values}
  n2 = MaterialFunctionCall() {materialFunction=&GetUserInterfaceUV}
  NamedRerouteDeclaration(None=Divide(A=ComponentMask(None=@n2.Pixel Size) {r=true}, B=ComponentMask(None=@n2.Pixel Size) {g=true})) {name=Pixel Size Horizontal Ratio, nodeColor=<rgba 0.6632 0.1421 0.5267 1>}
  n3 = MaterialFunctionCall() {materialFunction=&GetUserInterfaceUV}
  NamedRerouteDeclaration(None=Divide(A=ComponentMask(None=@n3.Pixel Size) {g=true}, B=ComponentMask(None=@n3.Pixel Size) {r=true})) {name=Pixel Size Vertical Ratio, nodeColor=<rgba 0 0.3437 1 1>}
  n4 = Multiply(A=NamedRerouteUsage()) {constB=0.01}
  n5 = Add(A=Multiply(A=@n4) {constB=0.5, desc=GPU : Using Multiply instead a Divide will use the GPU instead the CPU}, B=@n4)
  n6 = Subtract(A=@n4, B=Multiply(A=@n4, B=ScalarParameter() {defaultValue=0.9, parameterName=FeatherMin, sliderMax=1}))
  NamedRerouteDeclaration(None=Clamp(None=Add(A=@n5, B=@n6)) {maxDefault=0.09}) {name="FeatherMin Compensation Value ", nodeColor=<rgba 0 0 0 1>}
  n7 = ScalarParameter() {defaultValue=0.9, parameterName=FeatherMin, sliderMax=1}
  n8 = NamedRerouteUsage()
  NamedRerouteDeclaration(None=Clamp(None=Add(A=@n7, B=@n8)) {maxDefault=0.9999}) {name=Adjusted FeatherMin, nodeColor=<rgba 0.9568 1 0 1>}
  n9 = TextureCoordinate()
  n10 = OneMinus(None=ComponentMask(None=@n9) {r=true})
  n11 = ComponentMask(None=@n9) {g=true}
  n12 = DotProduct(A=StaticSwitch(True=AppendVector(A=@n10, B=@n11), False=@n9, Value=StaticBoolParameter() {parameterName=MirrorX}), B=@n1) {desc="The dot product gets you the angle between two vectors. \r\nbasically it'll give you a -1 -1 value :\r\nOr if the two vectors are parallel Or how perpendicular they are.\r\nAnd we are using that to get the distance between one edge (vector A) \r\nand the other edge (1) to use to create an angle."}
  NamedRerouteDeclaration(None=@n12) {name=Skew Angle, nodeColor=<rgba 0 0 0 1>}
  n13 = NamedRerouteUsage()
  n14 = NamedRerouteUsage()
  n15 = ScalarParameter() {defaultValue=1, parameterName="FeatherMax (Length)", sliderMax=1}
  n16 = StaticSwitchParameter(
   True=Divide(A=Subtract(A=@n13, B=@n14), B=Subtract(A=@n15, B=@n14))
   False=SmoothStep(Min=NamedRerouteUsage(), Max=ScalarParameter() {defaultValue=1, parameterName="FeatherMax (Length)", sliderMax=1}, Value=NamedRerouteUsage())
  ) {
   defaultValue=true
   desc="Alternate between [SmoothStep] and [Substract/Divide] \r\ncalculation method to render the anti-aliasing. \r\nThe result is slightly different. "
   parameterName=SwitchFeatherRendering
  }
  Saturate(None=OneMinus(None=@n16))
  Step(Y=NamedRerouteUsage(), X=Constant() {r=1})
 ]
}
